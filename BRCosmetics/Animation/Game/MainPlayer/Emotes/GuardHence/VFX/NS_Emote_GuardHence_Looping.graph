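# Looping Niagara effect for the GuardHence player emote, forming a dark screen-like backdrop with drifting glints and distortion.
# Screen uses spawn groups to split flickering mesh streaks from two loop-long sprites, which narrow by camera angle and dim by time of day.
# HangingParticulates spawns short-lived yellow-green specks, while Chromatic feeds a system undulate value into its sprite material.
# Head-attached glint emitters and a spotlight are disabled.
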NiagaraSystem /BRCosmetics/Animation/Game/MainPlayer/Emotes/GuardHence/VFX/NS_Emote_GuardHence_Looping
dependencies {
 usedDataInterfaces [
  &NiagaraDataInterfaceArrayFloat3
  &NiagaraDataInterfaceCurve
  &NiagaraDataInterfaceCamera
  &NiagaraDataInterfaceSkeletalMesh
 ]
 usedDynamicInputs [&MakeCustomVectorFromBool, &RandomBool]
 usedModules [
  &SystemState
  &self:SystemSpawnScript.SystemScriptSource.SystemScriptGraph.NiagaraNodeAssignment_0.SetVariables_#1
 ]
}
emitters {
 Chromatic {
  emitterUpdateScript [
   &EmitterState()
   &SpawnBurst_Instantaneous(Use Spawn Probability=false, UseLoopCountLimit=false)
  ]
  particleSpawnScript [
   &InitializeParticle(Lifetime=10, Sprite Size=<270 230>, Sprite Size Mode=::NewEnumerator4, UsePositionOffset=false, Write Material Random=true)
   &CameraOffset(Camera Offset Amount=100)
  ]
  particleUpdateScript [
   &ParticleState()
   /SetVariables(Particles.SpriteAlignment=<0 0 1>) {isSetParameters=true}
   &DynamicMaterialParameters(
    Index 0 Param 1=$System.undulate
    Index 0 Param 4=1
    Module.Index0Float0Write=true
    Module.Index0Float1Write=true
    Module.Index0Float2Write=true
    Module.Index0Float3Write=true
   )
   &DynamicMaterialParameters(
    Index 0 Param 1=&Waveform("Waveform [1]"=::NewEnumerator8, "[1] Frequency"=1, "[1] Amplitude Min/Max"=<0 0.6>)
    Index 0 Param 4=1
    Module.Index0Float0Write=true
    Module.Index0Float1Write=true
    Module.Index0Float2Write=true
    Module.Index0Float3Write=true
   ) {enabled=false, module=DynamicMaterialParameters001}
  ]
  rendererClasses [&NiagaraSpriteRendererProperties]
  renderers [{rendererClass=&NiagaraSpriteRendererProperties}]
 }
 Glows_Air {rendererClasses [], renderers []}
 HangingParticulates {
  emitterUpdateScript [
   &EmitterState(Life Cycle Mode=::NewEnumerator1, Loop Duration=2, UseLoopDelay=false)
   &SpawnRate(SpawnRate=25, Use Spawn Probability=false)
  ]
  particleSpawnScript [
   &InitializeParticle(
    Color Maximum=<rgba 1.7716 2 0.2507 1>
    Color Minimum=<rgba 0.3019 2 0.3586 1>
    Color Mode=::NewEnumerator2
    Lifetime Max=1
    Lifetime Min=0.5
    Lifetime Mode=::NewEnumerator1
    Sprite Rotation Mode=::NewEnumerator2
    Sprite Size Mode=::NewEnumerator3
    Sprite UV Mode=::NewEnumerator3
    Uniform Sprite Size=/Niagara/DynamicInputs/Helpers/DistributionCurveFloat(
     Minimum:NiagaraFloat=2.5
     Maximum:NiagaraFloat=17
     Distribution Curve:NiagaraDataInterfaceCurve=DataInterface() {propertyValues {Curve {
        defaultValue=0
        keys [
         {tangentMode=RCTM_User, time=0, value=0}
         <at 0.25 0.25>
         <at 0.5448 0.2934>
         <at 0.6534 0.6153>
         {tangentMode=RCTM_User, time=0.9107, value=1}
         <at 0.9627 1.4837>
        ]
       }}}
     Curve Parameter Type:NiagaraInt32=0
     "Curve Parameter Type:":NiagaraInt32=0
     Curve Parameter:NiagaraFloat=?
     Evaluation Type:ENiagaraRandomnessEvaluation=::NewEnumerator0
     Recalculate Random Each Loop:NiagaraBool=true
     Randomness Mode:ENiagaraRandomnessMode=::NewEnumerator0
     Random Seed:NiagaraInt32=0
     Fixed Random Seed:NiagaraBool=false
    )
    UsePositionOffset=false
    Write Material Random=true
   )
   &ShapeLocation(
    Box Midpoint=&RandomRangeVector(Minimum=<0 0 0>, Maximum=<1 1 1>)
    Box Size=<85 30 100>
    MaskSpawnGroup=false
    Offset=<0 -70 100>
    Offset Mode=::NewEnumerator0
    Override Seed=false
    Shape Primitive=::NewEnumerator2
    Surface Only=false
    UseOwnerScale=true
   )
   /SetVariables(Particles.VisibilityTag=&RandomRangeInt(Minimum=0, Maximum=2)) {isSetParameters=true}
  ]
  particleUpdateScript [
   &ParticleState()
   &ScaleSpriteSize(Scale Factor=&Multiply_Vector2D(
     A=&Vector2DFromFloat(Value=&RandomRangeFloat(
       Minimum=&FloatFromCurve(FloatCurve=DataInterface() {propertyValues {Curve {keys [
            <at 0 0 RCIM_Cubic>
            {arriveTangent=3.8811, interpMode=RCIM_Cubic, leaveTangent=3.8811, time=0.0665, value=1.0532}
            {arriveTangent=-1.7447, interpMode=RCIM_Cubic, leaveTangent=-1.7447, tangentMode=RCTM_User, time=0.163, value=0.6326}
            <at 1.0011 0.002>
           ]}}})
       Maximum=&FloatFromCurve(FloatCurve=DataInterface() {propertyValues {Curve {keys [
            <at 0 0 RCIM_Cubic>
            {arriveTangent=1.4708, interpMode=RCIM_Cubic, leaveTangent=1.4708, time=0.2262, value=1.0153}
            {arriveTangent=-1.7447, interpMode=RCIM_Cubic, leaveTangent=-1.7447, tangentMode=RCTM_User, time=0.2928, value=0.4306}
            <at 1.0011 0.002>
           ]}}})
      ))
     B=<1 1>
    ))
   &ScaleColor(
    Scale Alpha=&FloatFromCurve(FloatCurve=DataInterface() {propertyValues {Curve {keys [
         {arriveTangent=27.4026, interpMode=RCIM_Cubic, leaveTangent=27.4026, tangentMode=RCTM_User, time=0, value=0}
         {arriveTangent=2.9638, interpMode=RCIM_Cubic, leaveTangent=2.9638, time=0.0433, value=0.9882}
         {arriveTangent=-1.033, interpMode=RCIM_Cubic, leaveTangent=-1.033, time=0.1259, value=0.3733}
         <at 1 0 RCIM_Cubic>
        ]}}})
    ScaleA=true
    ScaleRGB=true
   )
   &SolveForcesAndVelocity(Clamp Velocity=false, Limit Acceleration=false)
  ]
  rendererClasses [&NiagaraSpriteRendererProperties]
  renderers [
   {rendererClass=&NiagaraSpriteRendererProperties}
   {rendererClass=&NiagaraSpriteRendererProperties, rendererIndex=1}
   {rendererClass=&NiagaraSpriteRendererProperties, rendererIndex=2}
  ]
 }
 Screen {
  emitterSpawnScript [/SetVariables(Emitter.U Scale Power:NiagaraFloat=&Waveform("Waveform [1]"=::NewEnumerator0, "[1] Frequency"=20, "[1] Amplitude Min/Max"=<0.5 0.6>)) {isSetParameters=true}]
  emitterUpdateScript [
   &EmitterState()
   &SpawnBurst_Instantaneous(Spawn Time=0.05, Use Spawn Probability=false, UseLoopCountLimit=false)
   &SpawnRate(SpawnRate=2.75, Use Spawn Probability=false)
   &SpawnBurst_Instantaneous(Spawn Group=1, Spawn Time=0.01, Use Spawn Probability=false, UseLoopCountLimit=false) {module=SpawnBurst_Instantaneous001}
   &SpawnBurst_Instantaneous(Spawn Group=2, Spawn Time=0.01, Use Spawn Probability=false, UseLoopCountLimit=false) {module=SpawnBurst_Instantaneous002}
  ]
  particleSpawnScript [
   /SetVariables(Particles.VisibilityTag=&SpawnGroup()) {isSetParameters=true}
   &InitializeParticle(
    Color=&MergeHelper_ColorSwitch(
     0=&RandomRangeLinearColor(Minimum=<rgba 0.4814 0.1514 0.9688 0.7>, Maximum=<rgba 0.0148 0.0057 0.0885 0.63>)
     1=<rgba 0.0205 0.0095 0.0521 1>
     2=<rgba 0.013 0.0677 0 0.1739>
     3=?
     Index=&SpawnGroup()
     Num=::NewEnumerator1
    )
    Lifetime=&MergeHelper_FloatSwitch(
     0=&RandomRangeFloat(Minimum=0.7, Maximum=1)
     1=$Emitter.CurrentLoopDuration
     2=$Emitter.CurrentLoopDuration
     3=?
     Index=&SpawnGroup()
     Num=::NewEnumerator1
    )
    Mesh Scale=&RandomRangeVector(Minimum=<4 0.05 3>, Maximum=<7 0.05 5>)
    Mesh Scale Mode=::NewEnumerator4
    Position Offset=&MergeHelper_VectorSwitch(
     0=&RandomRangeVector(Minimum=<-33 -80 75>, Maximum=<33 -80 80>)
     1=<0 -80 90>
     2=<0 -80 90>
     3=?
     Index=&SpawnGroup()
     Num=::NewEnumerator1
    )
    Sprite Size=&MergeHelper_Vector2Switch(0=<0 0>, 1=<500 350>, 2=<150 250>, 3=?, 4=?, Index=&SpawnGroup(), Num=::NewEnumerator1)
    Sprite Size Mode=::NewEnumerator4
    UsePositionOffset=true
    Write Material Random=true
   )
   &SpriteFacingAndAlignment(Sprite Facing=<0 1 0>, WriteAlignment=false, WriteFacing=true)
   &ShapeLocation(
    Box Size=&MergeHelper_VectorSwitch(0=<5 15 130>, 1=<0 0 0>, 2=<0 0 0>, 3=?, Index=&SpawnGroup(), Num=::NewEnumerator1)
    MaskSpawnGroup=false
    Override Seed=false
    Shape Primitive=::NewEnumerator2
    Surface Only=false
    UseOwnerScale=true
   )
   &InitialMeshOrientation(
    Mesh Orientation Mode=::NewEnumerator2
    Orientation Vector=&SelectVectorFromArray(
     Array Sampling Mode=::NewEnumerator0
     Direct Array Index=?
     Vector Selection Array=DataInterface() {propertyValues {FloatData [<1 0 0>, <-1 0 0>]}}
     Random Seed=0
    )
    Use Rotation Vector=false
   )
  ]
  particleUpdateScript [
   &ParticleState()
   &ScaleSpriteSize(
    Scale Factor=&Multiply_Vector2D(
     A=<1 1>
     B=&MakeVector2D(
      X=&Multiply_Float(
       A=1
       B=&RemapRange(
        Input Value To Remap=&Abs_Float(Float:NiagaraFloat=&DotProduct(
          Vector A=&ReturnCameraVector(Camera Vectors=::NewEnumerator4, Coordinate Space=::Simulation, Origin Position=$Particles.Position)
          Vector B=$Particles.SpriteFacing
         ))
        Input Min=0.2
        Input Max=0.9
        Output Min=&MergeHelper_FloatSwitch(0=0, 1=0.4, 2=0.26, 3=?, Index=&SpawnGroup(), Num=::NewEnumerator1)
        Output Max=1
       )
      )
      Y=1
     )
    )
    Scale Sprite Size Mode=::NewEnumerator2
   )
   &SolveForcesAndVelocity(Clamp Velocity=false, Limit Acceleration=false)
   &ScaleColor(
    Scale Alpha=&MergeHelper_FloatSwitch(
     0=&FloatFromCurve(
      FloatCurve=DataInterface() {propertyValues {Curve {
         defaultValue=1
         keys [
          {interpMode=RCIM_Cubic, tangentMode=RCTM_User, time=0, value=0}
          {interpMode=RCIM_Cubic, tangentMode=RCTM_User, time=0.5, value=1}
          {tangentMode=RCTM_User, time=1, value=0}
         ]
        }}}
      Scale Curve=&FNTimeOfDayScalar(Night Scale=0.7, Morning Scale=0.8, Evening Scale=0.8)
     )
     1=&FNTimeOfDayScalar(Night Scale=0.65, Morning Scale=0.7, Evening Scale=0.7)
     2=&FNTimeOfDayScalar(Night Scale=0.65, Morning Scale=0.7, Evening Scale=0.7)
     3=?
     Index=&SpawnGroup()
     Num=::NewEnumerator1
    )
    ScaleA=true
    ScaleRGB=true
   )
   /SetVariables(Particles.Linear Color=<rgba 0.0022 0.0104 0 1>) {isSetParameters=true}
  ]
  rendererClasses [&NiagaraMeshRendererProperties, &NiagaraSpriteRendererProperties]
  renderers [
   {rendererClass=&NiagaraMeshRendererProperties, rendererIndex=0}
   {rendererClass=&NiagaraSpriteRendererProperties, rendererIndex=1}
   {rendererClass=&NiagaraSpriteRendererProperties, rendererIndex=2}
   {rendererClass=&NiagaraSpriteRendererProperties, rendererIndex=3}
  ]
 }
 SingleLoopingParticle {
  bEnabled=false
  emitterUpdateScript [
   &EmitterState()
   &SpawnBurst_Instantaneous(Spawn Time=3.35, Use Spawn Probability=false, UseLoopCountLimit=false)
  ]
  particleSpawnScript [
   &InitializeParticle(
    Color=<rgba 0.9947 1 0.3104 0.2>
    Lifetime=0.25
    Sprite Size Mode=::NewEnumerator3
    Uniform Sprite Size=300
    UsePositionOffset=false
    Write Material Random=true
   )
   &CameraOffset(Camera Offset Amount=20)
  ]
  particleUpdateScript [
   &ParticleState()
   &ScaleSpriteSize(Uniform Curve Sprite Scale=DataInterface() {propertyValues {Curve {
       defaultValue=0
       keys [
        {interpMode=RCIM_Cubic, tangentMode=RCTM_User, time=0, value=0.75}
        {interpMode=RCIM_Cubic, tangentMode=RCTM_User, time=0.05, value=1}
        {tangentMode=RCTM_User, time=1, value=0.75}
       ]
      }}})
   &SkeletalMeshLocation(
    Bone Sampling Mode=::NewEnumerator4
    MaskSpawnGroup=false
    Sampled Position Offset=<0 25 0>
    Skeletal Mesh=DataInterface() {propertyValues {
      DefaultMesh=null
      ExcludeBoneName=root
      OptionalComponentClass=null
      PreviewMesh=/Game/Characters/Player/Male/Medium/Bodies/M_Med_Soldier_01/Meshes/M_Med_Soldier_01
      SkinningMode=SkinOnTheFly
      SoftSourceActor=null
      SourceMode=Default
      UvSetIndex=0
      WholeMeshLOD=-1
      bExcludeBone=false
      bReadDeformedGeometry=true
      bRequireCurrentFrameData=true
      ComponentTags []
      FilteredBones [head]
      FilteredSockets []
      MeshUserParameter {parameter {name=None, typeDefHandle {registeredTypeIndex=1}, varData []}}
      SamplingRegions []
     }}
   )
   &SolveForcesAndVelocity(Clamp Velocity=false, Limit Acceleration=false)
   &ScaleColor(
    Scale Alpha=&FloatFromCurve(FloatCurve=DataInterface() {propertyValues {Curve {
        defaultValue=1
        keys [
         {arriveTangent=9.0725, interpMode=RCIM_Cubic, leaveTangent=9.0725, tangentMode=RCTM_User, time=0, value=0}
         {arriveTangent=-2.5249, interpMode=RCIM_Cubic, leaveTangent=-2.5249, tangentMode=RCTM_User, time=0.05, value=1}
         {arriveTangent=-1.3198, interpMode=RCIM_Cubic, leaveTangent=-1.3198, tangentMode=RCTM_User, time=0.1085, value=0.4976}
         {arriveTangent=-0.0405, interpMode=RCIM_Cubic, leaveTangent=-0.0405, tangentMode=RCTM_User, time=1, value=0}
        ]
       }}})
    ScaleA=true
    ScaleRGB=true
   )
  ]
  rendererClasses [&NiagaraSpriteRendererProperties]
  renderers [{rendererClass=&NiagaraSpriteRendererProperties}]
 }
 SingleLoopingParticle001 {
  bEnabled=false
  emitterUpdateScript [
   &EmitterState()
   &SpawnBurst_Instantaneous(Spawn Time=2.35, Use Spawn Probability=false, UseLoopCountLimit=false)
  ]
  particleSpawnScript [
   &InitializeParticle(
    Color=<rgba 0.9947 1 0.3104 0.01>
    Lifetime=2
    Sprite Size Mode=::NewEnumerator3
    Uniform Sprite Size=150
    UsePositionOffset=false
    Write Material Random=true
   )
   &CameraOffset(Camera Offset Amount=20)
  ]
  particleUpdateScript [
   &ParticleState()
   &ScaleSpriteSize() {enabled=false, module=ScaleSpriteSize001}
   &ScaleSpriteSize(Uniform Curve Sprite Scale=DataInterface() {propertyValues {Curve {
       defaultValue=0.75
       keys [
        {interpMode=RCIM_Cubic, tangentMode=RCTM_User, time=0, value=0}
        {arriveTangent=2.0967, interpMode=RCIM_Cubic, leaveTangent=2.0967, time=0.0508, value=0.404}
        {arriveTangent=-1.02, interpMode=RCIM_Cubic, leaveTangent=-1.02, time=0.1033, value=0.2165}
        {arriveTangent=5.358, interpMode=RCIM_Cubic, leaveTangent=5.358, time=0.1584, value=0.2942}
        {
         arriveTangent=3.618
         arriveTangentWeight=0.0412
         interpMode=RCIM_Cubic
         leaveTangent=3.618
         leaveTangentWeight=0.1487
         tangentMode=RCTM_User
         time=0.3659
         value=1.6235
        }
        {arriveTangent=1.1939, interpMode=RCIM_Cubic, leaveTangent=1.1939, time=0.474, value=2}
        {
         arriveTangent=-4.0035
         arriveTangentWeight=0.4402
         interpMode=RCIM_Cubic
         leaveTangent=-4.0035
         leaveTangentWeight=0.2681
         tangentMode=RCTM_User
         time=0.5827
         value=1.8824
        }
        {arriveTangent=-5.4368, interpMode=RCIM_Cubic, leaveTangent=-5.4368, tangentMode=RCTM_User, time=0.6568, value=0.1869}
        {arriveTangent=-0.5445, interpMode=RCIM_Cubic, leaveTangent=-0.5445, time=0.8518, value=-0.0312}
        {tangentMode=RCTM_User, time=1, value=0}
       ]
      }}})
   &SkeletalMeshLocation(
    Bone Sampling Mode=::NewEnumerator4
    MaskSpawnGroup=false
    Sampled Position Offset=<-20 25 0>
    Skeletal Mesh=DataInterface() {propertyValues {
      DefaultMesh=null
      ExcludeBoneName=root
      OptionalComponentClass=null
      PreviewMesh=&M_Med_Soldier_01
      SkinningMode=SkinOnTheFly
      SoftSourceActor=null
      SourceMode=Default
      UvSetIndex=0
      WholeMeshLOD=-1
      bExcludeBone=false
      bReadDeformedGeometry=true
      bRequireCurrentFrameData=true
      ComponentTags []
      FilteredBones [head]
      FilteredSockets []
      MeshUserParameter {parameter {name=None, typeDefHandle {registeredTypeIndex=1}, varData []}}
      SamplingRegions []
     }}
   )
   &SolveForcesAndVelocity(Clamp Velocity=false, Limit Acceleration=false)
   &ScaleColor(
    Scale Alpha=&FloatFromCurve(FloatCurve=DataInterface() {propertyValues {Curve {
        defaultValue=0
        keys [
         {interpMode=RCIM_Cubic, tangentMode=RCTM_User, time=0, value=0}
         {interpMode=RCIM_Cubic, tangentMode=RCTM_User, time=0.5, value=1}
         {arriveTangent=-4.1738, interpMode=RCIM_Cubic, leaveTangent=-4.1738, time=0.6, value=0.8999}
         {arriveTangent=-2.2499, interpMode=RCIM_Cubic, leaveTangent=-2.2499, time=0.7035, value=0.1505}
         {tangentMode=RCTM_User, time=1, value=0}
        ]
       }}})
    ScaleA=true
    ScaleRGB=true
   )
   /SetVariables(Particles.PivotOffset:Vector2f=&MakeVector2D(
     X=&FloatFromCurve(FloatCurve=DataInterface() {propertyValues {Curve {keys [<at 0 -1>,<at 0.47 0.5>,<at 0.75 0.5>,<at 1 2>]}}})
     Y=&FloatFromCurve(FloatCurve=DataInterface() {propertyValues {Curve {keys [<at 0 -1>,<at 0.47 0.5>,<at 0.75 0.5>,<at 0.9795 -4>]}}})
    )) {isSetParameters=true}
  ]
  rendererClasses [&NiagaraSpriteRendererProperties]
  renderers [{rendererClass=&NiagaraSpriteRendererProperties}]
 }
 spotlight {bEnabled=false, rendererClasses [], renderers []}
}
summary {
 systemName=NS_Emote_GuardHence_Looping
 emitters [
  {emitterName=Chromatic, rendererClasses [&NiagaraSpriteRendererProperties]}
  {emitterName=Screen, rendererClasses [&NiagaraMeshRendererProperties, &NiagaraSpriteRendererProperties]}
  {emitterName=Glows_Air, rendererClasses []}
  {emitterName=HangingParticulates, rendererClasses [&NiagaraSpriteRendererProperties]}
  {bEnabled=false, emitterName=SingleLoopingParticle, rendererClasses [&NiagaraSpriteRendererProperties]}
  {bEnabled=false, emitterName=spotlight, rendererClasses []}
  {bEnabled=false, emitterName=SingleLoopingParticle001, rendererClasses [&NiagaraSpriteRendererProperties]}
 ]
 userVariables []
}
